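# Masked material for the ThawEpson character, with four ID-selected body colors and moving, scared, dead, aggro and attack states.
# UV tiles tag body, two leg sets, eyes, pupils and scared parts; state parameters and looping time mask and displace these, animating walking and eye direction.
# Start-time parameters shape aggro and attack pulses for emissive, alongside time-of-day fake top light.
# An IsOutline switch creates black, distance-scaled inverted-hull outlines.
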Material /ThawEpsonAI/Assets/TechArt/M_ThawEpson
graph {
 properties {bAutomaticallySetUsageInEditor=false, blendMode=BLEND_Masked, usedShadingModels=Default Lit}
 nodes [
  ComponentMask(None=TextureCoordinate()) {r=true}
  Frac(None=Divide(A=Time(), B=ScalarParameter() {defaultValue=2, group=CosmicThunder-AnimationCurve, parameterName=ThawEpson_Time_Preview_Loop_Duration}))
  Frac(None=Divide(A=Time(), B=ScalarParameter() {defaultValue=2, group=CosmicThunder-AnimationCurve, parameterName=ThawEpson_Time_Preview_Loop_Duration}))
  n1 = WorldPosition()
  n2 = ScalarParameter() {defaultValue=16}
  Frac(None=Divide(A=@n1.Z, B=@n2))
  n3 = NamedRerouteUsage()
  n4 = NamedRerouteUsage()
  n5 = Saturate(None=Add(A=Add(A=@n3, B=@n4), B=NamedRerouteUsage()))
  n6 = NamedRerouteUsage()
  n7 = Constant()
  n8 = Constant3Vector() {precision=Half}
  n9 = Constant3Vector() {constant=<rgba 1 0 0 1>}
  n10 = Constant() {r=1}
  n11 = Constant3Vector() {constant=<rgba 1 0.3916 0.7305 1>}
  n12 = Add(
   A=Add(A=If(A=@n6, B=@n7, "A > B"=@n8, "A == B"=@n9, "A < B"=@n8), B=If(A=@n6, B=@n10, "A > B"=@n8, "A == B"=@n11, "A < B"=@n8) {precision=Half})
   B=If(A=@n6, B=Constant() {r=2}, "A > B"=@n8, "A == B"=Constant3Vector() {constant=<rgba 0 0.6105 1 1>}, "A < B"=@n8) {precision=Half}
  )
  n13 = If(A=@n6, B=Constant() {r=3}, "A > B"=@n8, "A == B"=Constant3Vector() {constant=<rgba 1 0.3916 0 1>}, "A < B"=@n8) {precision=Half}
  n14 = NamedRerouteDeclaration(None=LinearInterpolate(A=Add(A=@n12, B=@n13), B=Constant3Vector() {constant=<rgba 0.0382 0.0319 1 1>}, Alpha=NamedRerouteUsage())) {name=EpsonThaw_Body_Color, nodeColor=<rgba 1 0.8486 0 1>}
  n15 = LinearInterpolate(
   A=Multiply(A=VectorParameter() {parameterName=Eye Color, defaultValue=<rgba 0.7454 0.7231 1 1>}.RGB, B=NamedRerouteUsage())
   B=VectorParameter() {parameterName=Pupil Color, defaultValue=<rgba 0.0152 0.0152 1 1>}.RGB
   Alpha=Multiply(A=NamedRerouteUsage(), B=NamedRerouteUsage())
  )
  n16 = OneMinus(None=NamedRerouteUsage())
  n17 = LinearInterpolate(
   A=Add(A=Multiply(A=@n5, B=@n14), B=Multiply(A=@n15, B=@n16))
   B=VectorParameter() {parameterName=Scared Color, defaultValue=<rgba 1 0.4793 0.3095 1>}.RGB
   Alpha=Multiply(A=NamedRerouteUsage(), B=NamedRerouteUsage())
  )
  LinearInterpolate(A=@n17, B=Constant3Vector(), Alpha=OneMinus(None=VertexColor().G))
  LocalPosition()
  MaterialFunctionCall() {materialFunction=/Game/Packages/Fortress_SharedMaterials/Base_Material_Functions/SkylightColorAndIntensity.SkyLightColorAndIntensity}
  Min(A=NamedRerouteUsage(), B=NamedRerouteUsage())
  n18 = LocalPosition()
  n19 = ScalarParameter() {defaultValue=0.08, parameterName=Grid Scale}
  n20 = Frac(None=Multiply(A=ComponentMask(None=@n18.XYZ) {r=true}, B=@n19) {constB=0.08})
  n21 = Frac(None=Multiply(A=ComponentMask(None=@n18.XYZ) {b=true}, B=@n19) {constB=0.08})
  Multiply(A=Power(Base=Multiply(A=@n20, B=@n21)) {constExponent=0.5}, B=@n14)
  n22 = Saturate(None=NamedRerouteUsage())
  n23 = Sine(None=Multiply(A=@n22) {constB=3})
  n24 = Multiply(A=LinearInterpolate(Alpha=ConstantBiasScale(None=@n23)), B=OneMinus(None=@n22))
  Multiply(A=MaterialFunctionCall!(X range Y range=Constant4Vector() {constant=<rgba 0 1 0 1>}.RGBA, Gradient Height, ShowGrid?, Thickness, Color, "f(x)"=@n24) {materialFunction=/Engine/Functions/Engine_MaterialFunctions01/Debug/PlotFunctionOnGraph}."f(x) = y")
  Multiply(A=@n1.Z, B=@n2)
  n25 = ComponentMask(None=TextureCoordinate()) {r=true}
  n26 = Ceil(None=@n25)
  n27 = Saturate(None=Floor(None=@n25))
  NamedRerouteDeclaration(None=Subtract(A=Saturate(None=@n26), B=@n27)) {name=Body, nodeColor=<rgba 1 0.2344 0 1>}
  n28 = Constant() {r=1}
  n29 = Saturate(None=Floor(None=Subtract(A=@n25, B=@n28)))
  NamedRerouteDeclaration(None=Subtract(A=@n27, B=@n29)) {name=Legs_1, nodeColor=<rgba 1 0.2344 0 1>}
  NamedRerouteDeclaration(None=ScalarParameter() {parameterName=ThawEpson_IsDead}) {name=ThawEpson_IsDead, nodeColor=<rgba 1 0 0.212 1>}
  NamedRerouteDeclaration(None=ScalarParameter() {desc="0 = Left, 1 = Right", parameterName=ThawEpson_Eyes_Left_Right}) {name=ThawEpson_Eyes_Left_Right, nodeColor=<rgba 1 0 0.212 1>}
  NamedRerouteDeclaration(None=ScalarParameter() {defaultValue=0.75, parameterName=ThawEpson_Duration_Aggro}) {name=ThawEpson_Duration_Aggro, nodeColor=<rgba 0.6875 0 1 1>}
  NamedRerouteDeclaration(None=ScalarParameter() {defaultValue=0.25, parameterName=ThawEpson_Duration_Attack}) {name=ThawEpson_Duration_Attack, nodeColor=<rgba 0.6875 0 1 1>}
  n30 = Subtract(A=Time(), B=NamedRerouteUsage())
  n31 = NamedRerouteUsage()
  NamedRerouteDeclaration(None=Saturate(None=Divide(A=@n30, B=@n31))) {name=TimestampSined_Aggro, nodeColor=<rgba 0 0.0625 1 1>}
  n32 = SmoothStep(
   Max=ScalarParameter() {defaultValue=0.15, group=CosmicThunder-AnimationCurve, parameterName=ThawEpson_Aggro_Intro_Duration_%}
   Value=NamedRerouteUsage()
  )
  n33 = ScalarParameter() {defaultValue=4, group=CosmicThunder-AnimationCurve, parameterName=ThawEpson_Aggro_Intro_Pow}
  NamedRerouteDeclaration(None=Saturate(None=Power(Base=@n32, Exp=@n33) {constExponent=4})) {name=Time_Intro_Aggro, nodeColor=<rgba 1 0 0.7969 1>}
  NamedRerouteDeclaration(None=ScalarParameter() {desc="0 = Red, 1 = Pink, 2 = Cyan, 3 = Orange", parameterName=ThawEpson_ColorID}) {name=ThawEpson_ColorID, nodeColor=<rgba 1 0.4068 0 1>}
  n34 = Constant() {r=2}
  n35 = Saturate(None=Floor(None=Subtract(A=@n25, B=@n34)))
  NamedRerouteDeclaration(None=Subtract(A=@n29, B=@n35)) {name=Legs_2, nodeColor=<rgba 1 0.2344 0 1>}
  n36 = ScalarParameter() {defaultValue=0.8, group=CosmicThunder-AnimationCurve, parameterName=ThawEpson_Aggro_Outro_Duration_%}
  n37 = OneMinus(None=SmoothStep(Min=OneMinus(None=@n36), Value=NamedRerouteUsage()))
  n38 = ScalarParameter() {defaultValue=8, group=CosmicThunder-AnimationCurve, parameterName=ThawEpson_Aggro_Outro_Pow}
  NamedRerouteDeclaration(None=Saturate(None=Power(Base=@n37, Exp=@n38) {constExponent=4})) {name=Time_Outro_Aggro, nodeColor=<rgba 1 0 0.4453 1>}
  n39 = Subtract(A=Time(), B=NamedRerouteUsage())
  n40 = NamedRerouteUsage()
  NamedRerouteDeclaration(None=Saturate(None=Divide(A=@n39, B=@n40))) {name=TimestampSined_Attack, nodeColor=<rgba 0 0.0625 1 1>}
  n41 = SmoothStep(
   Max=ScalarParameter() {defaultValue=0.05, group=CosmicThunder-AnimationCurve, parameterName=ThawEpson_Attack_Intro_Duration_%}
   Value=NamedRerouteUsage()
  )
  n42 = ScalarParameter() {defaultValue=4, group=CosmicThunder-AnimationCurve, parameterName=ThawEpson_Attack_Intro_Pow}
  NamedRerouteDeclaration(None=Saturate(None=Power(Base=@n41, Exp=@n42) {constExponent=4})) {name=Time_Intro_Attack, nodeColor=<rgba 1 0 0.7969 1>}
  n43 = ScalarParameter() {defaultValue=0.25, group=CosmicThunder-AnimationCurve, parameterName=ThawEpson_Attack_Outro_Duration_%}
  n44 = OneMinus(None=SmoothStep(Min=OneMinus(None=@n43), Value=NamedRerouteUsage()))
  n45 = ScalarParameter() {defaultValue=8, group=CosmicThunder-AnimationCurve, parameterName=ThawEpson_Attack_Outro_Pow}
  NamedRerouteDeclaration(None=Saturate(None=Power(Base=@n44, Exp=@n45) {constExponent=4})) {name=Time_Outro_Attack, nodeColor=<rgba 1 0 0.4453 1>}
  n46 = LinearInterpolate(A=NamedRerouteUsage(), B=NamedRerouteUsage(), Alpha=NamedRerouteUsage())
  NamedRerouteDeclaration(None=@n46) {name=Eyes_Left_Right_Masked, nodeColor=<rgba 0.4062 1 0 1>}
  NamedRerouteDeclaration(None=@n24) {name=TimeStamp_Aggro, nodeColor=<rgba 0.8281 1 0 1>}
  NamedRerouteDeclaration(None=ScalarParameter() {parameterName=ThawEpson_StartTime_Attack}) {name=ThawEpson_StartTime_Attack, nodeColor=<rgba 0.6875 0 1 1>}
  NamedRerouteDeclaration(None=Min(A=NamedRerouteUsage(), B=NamedRerouteUsage())) {name=TimeStamp_Attack, nodeColor=<rgba 0.8281 1 0 1>}
  n47 = Constant() {r=3}
  n48 = Saturate(None=Floor(None=Subtract(A=@n25, B=@n47)))
  NamedRerouteDeclaration(None=Subtract(A=@n35, B=@n48)) {name=Eyes_Left, nodeColor=<rgba 1 0.2344 0 1>}
  n49 = Constant() {r=4}
  n50 = Saturate(None=Floor(None=Subtract(A=@n25, B=@n49)))
  NamedRerouteDeclaration(None=Subtract(A=@n48, B=@n50)) {name=Eyes_Right, nodeColor=<rgba 1 0.2344 0 1>}
  NamedRerouteDeclaration(None=@n50) {name=Scared, nodeColor=<rgba 1 0.2344 0 1>}
  n51 = Ceil(None=ComponentMask(None=TextureCoordinate()) {g=true})
  NamedRerouteDeclaration(None=OneMinus(None=Saturate(None=@n51))) {name=Pupils, nodeColor=<rgba 1 0.2344 0 1>}
  NamedRerouteDeclaration(None=ScalarParameter() {parameterName=ThawEpson_StartTime_Aggro}) {name=ThawEpson_StartTime_Aggro, nodeColor=<rgba 0.6875 0 1 1>}
  NamedRerouteDeclaration(None=ScalarParameter() {defaultValue=1, parameterName=ThawEpson_IsMoving}) {name=ThawEpson_IsMoving, nodeColor=<rgba 1 0 0.212 1>}
  NamedRerouteDeclaration(None=ScalarParameter() {parameterName=ThawEpson_IsScared}) {name=ThawEpson_IsScared, nodeColor=<rgba 1 0 0.212 1>}
  NamedRerouteUsage()
  NamedRerouteUsage()
  NamedRerouteUsage()
  NamedRerouteUsage()
  NamedRerouteUsage()
  NamedRerouteUsage()
  NamedRerouteUsage()
  NamedRerouteUsage()
  NamedRerouteUsage()
  NamedRerouteUsage()
  NamedRerouteUsage()
  NamedRerouteUsage()
  n52 = Constant3Vector() {constant=<rgba 0 0 1 0>}
  Normalize(VectorInput=Add(A=AtmosphericLightVector(), B=@n52))
  ObjectPositionWS() {originType=CameraRelative}
  OneMinus()
  n53 = Constant3Vector()
  n54 = ObjectPositionWS() {originType=CameraRelative}
  n55 = ScalarParameter() {defaultValue=2000, parameterName=Outline Scale Distance}
  n56 = LinearInterpolate(
   A=ScalarParameter() {defaultValue=5, parameterName=Outline Offset Max}
   B=ScalarParameter() {defaultValue=1, parameterName=Outline Offset Min}
   Alpha=Saturate(None=SphereMask(A=@n53, B=@n54, Radius=@n55) {hardnessPercent=1})
  )
  n57 = NamedRerouteUsage()
  n58 = Frac(None=Divide(A=Time(), B=ScalarParameter() {defaultValue=0.75, parameterName=Walk Anim Loop Duration}))
  n59 = LinearInterpolate(A=NamedRerouteUsage(), B=NamedRerouteUsage(), Alpha=Multiply(A=Step(Y=@n58) {constX=0.5}, B=NamedRerouteUsage()))
  n60 = NamedRerouteUsage()
  n61 = NamedRerouteUsage()
  n62 = OneMinus(None=@n61)
  n63 = NamedRerouteUsage()
  n64 = Add(A=Multiply(A=Add(A=@n57, B=@n59), B=OneMinus(None=@n60)), B=Add(A=Multiply(A=@n46, B=@n62), B=Multiply(A=@n63, B=@n61)))
  n65 = Multiply(A=OneMinus(None=@n64), B=ObjectPositionWS())
  n66 = Add(A=Multiply(A=VertexNormalWS(), B=Multiply(A=@n56) {constB=-1}), B=@n65)
  n67 = CameraPositionWS()
  n68 = WorldPosition()
  n69 = Multiply(
   A=Normalize(VectorInput=Subtract(A=@n67, B=@n68.XYZ))
   B=ScalarParameter() {defaultValue=0.5, group=01_InvertedHull_Outlines, parameterName=CameraDirectionOffset, sliderMax=2, sortPriority=3}
  )
  PreviousFrameSwitch(Current Frame=StaticSwitchParameter(True=Add(A=@n66, B=@n69), False=@n65) {parameterName=IsOutline}, Previous Frame=Constant())
  ScalarParameter() {group=CosmicThunder-AnimationCurve, parameterName=ThawEpson_Time_Preview_FixedTime}
  ScalarParameter() {defaultValue=0.3, parameterName=Roughness}
  ScalarParameter() {defaultValue=0.5, parameterName=Metallic}
  ScalarParameter() {group=CosmicThunder-AnimationCurve, parameterName=ThawEpson_Time_Preview_FixedTime}
  StaticSwitchParameter(True=Constant3Vector(), False=@n17) {parameterName=IsOutline}
  StaticSwitchParameter(True=Multiply(A=Constant() {r=1}, B=@n64), False=@n64) {parameterName=IsOutline}
  n70 = LinearInterpolate(
   B=Multiply(A=ScalarParameter() {defaultValue=8, parameterName=State Emissive Strength}, B=NamedRerouteUsage())
   Alpha=Add(A=NamedRerouteUsage(), B=NamedRerouteUsage())
  )
  n71 = VectorParameter() {parameterName="Fake Zup Light Intensity (TOD)", channelNames {a=Night, b=Evening, g=Day, r=Morning}, defaultValue=<rgba 3.5 5 3.5 1>}
  n72 = Multiply(
   A=@n17
   B=MaterialFunctionCall(Value_Morning=@n71.Morning, Value_Day=@n71.Morning, Value_Evening=@n71.Morning, Value_Night=@n71.Morning) {materialFunction=&Scalar_Time_of_Day}.Result
  )
  n73 = DotProduct(A=@n52, B=VertexNormalWS())
  StaticSwitchParameter(True=@n70, False=LinearInterpolate(A=@n70, B=Add(A=@n72, B=@n70), Alpha=Saturate(None=@n73))) {parameterName=IsOutline}
  Step()
  n74 = ComponentMask(None=TextureCoordinate()) {r=true}
  Subtract(A=@n74) {constB=0.1}
  Subtract(A=@n74, B=ScalarParameter() {defaultValue=-10, parameterName=BP_GetTime})
  Time()
 ]
}
